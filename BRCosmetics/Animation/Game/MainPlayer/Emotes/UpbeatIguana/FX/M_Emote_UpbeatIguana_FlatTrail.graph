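# Translucent unlit Niagara ribbon material for the flat trail in the UpbeatIguana emote.
# Dynamic parameters drive three contrast-adjustable color bands across the ribbon width or one color, blended with a base color and tinted by particle color.
# Opacity combines masks such as edge, length, instance-space radial and fresnel, with particle alpha shaping falloff through an exponent.
# A platform switch replaces depth fade with a constant on Switch, Android and PC performance mode.
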Material /BRCosmetics/Animation/Game/MainPlayer/Emotes/UpbeatIguana/FX/M_Emote_UpbeatIguana_FlatTrail
graph {
 properties {
  bAutomaticallySetUsageInEditor=false
  bUsedWithNiagaraRibbons=true
  blendMode=BLEND_Translucent
  pixelDepthOffsetMode=PDOM_Legacy
  shadingModel=MSM_Unlit
  usedShadingModels=Unlit
 }
 nodes [
  n1 = DynamicParameter() {paramNames [Color 0 R, Color 0 G, Color 0 B, Param4]}
  n2 = ComponentMask(None=TextureCoordinate()) {g=true}
  n3 = Multiply(A=@n2) {constB=4}
  n4 = Subtract(A=@n3)
  n5 = DynamicParameter() {parameterIndex=3, defaultValue=<rgba 0 4 1 1>, paramNames [Use Single Color, Color Lerp Contrast, Fresnel Exponent, Opacity Multiplier]}
  n6 = Multiply(A=NamedRerouteUsage(), B=@n5.Color Lerp Contrast)
  n7 = LinearInterpolate(
   A=@n1.RGB
   B=DynamicParameter() {parameterIndex=1, paramNames [Color 1 R, Color 1 G, Color 1 B, Param4]}.RGB
   Alpha=MaterialFunctionCall(In=Saturate(None=@n4), Contrast=@n6) {materialFunction=&CheapContrast}.Result
  )
  n8 = DynamicParameter() {parameterIndex=2, paramNames [Color 2 R, Color 2 G, Color 2 B, Param4]}
  n9 = Subtract(A=@n4)
  n10 = MaterialFunctionCall(In=Saturate(None=@n9), Contrast=@n6) {materialFunction=&CheapContrast}
  n11 = LinearInterpolate(
   A=ComponentMask(None=LinearInterpolate(A=@n7, B=@n8.RGB, Alpha=@n10.Result)) {b=true, g=true, r=true}
   B=@n1.RGB
   Alpha=Saturate(None=@n5.Use Single Color)
  )
  n12 = VectorParameter() {parameterName=Base Color, defaultValue=<rgba 1 1 1 0>}
  n13 = NamedRerouteUsage()
  n14 = ParticleColor()
  MaterialFunctionCall(Emissive=Multiply(A=LinearInterpolate(A=@n11, B=@n12.RGB, Alpha=@n13), B=@n14.RGB)) {materialFunction=&MF_FX_EmissiveAdjustments}
  n15 = OneMinus(None=ComponentMask(None=TextureCoordinate()) {r=true})
  n16 = ScalarParameter() {defaultValue=0.05, parameterName=Base Mask Size}
  n17 = Divide(A=Subtract(A=@n15, B=@n16), B=Subtract(B=@n16))
  n18 = Power(Base=OneMinus(None=Saturate(None=@n17)), Exp=ScalarParameter() {defaultValue=6, parameterName=Length Mask Exponent})
  n19 = Multiply(A=NamedRerouteUsage(), B=@n18)
  n20 = NamedRerouteUsage()
  n21 = Multiply(A=Power(Base=Multiply(A=@n19, B=@n20), Exp=Divide(B=@n14.A)), B=NamedRerouteUsage())
  n22 = NamedRerouteUsage()
  n23 = DynamicParameter() {parameterIndex=3, defaultValue=<rgba 0 4 1 1>, paramNames [Use Single Color, Color Lerp Contrast, Fresnel Exponent, Opacity Multiplier]}
  Multiply(A=Multiply(A=Multiply(A=@n21, B=@n22), B=NamedRerouteUsage()), B=@n23.Opacity Multiplier)
  n24 = WorldPosition()
  n25 = Length(None=ComponentMask(None=TransformPosition(None=@n24.XYZ) {transformSourceType=TRANSFORMPOSSOURCE_World, transformType=TRANSFORMPOSSOURCE_Instance}) {g=true, r=true}) {is=class MaterialExpressionLength}
  n26 = Subtract(A=@n25, B=ScalarParameter() {defaultValue=44, parameterName=Circular Mask Offset})
  n27 = ScalarParameter() {defaultValue=44, parameterName=Circular Mask Length}
  NamedRerouteDeclaration(None=Saturate(None=Divide(A=@n26, B=@n27))) {name=Circular Mask, nodeColor=<rgba 0.1484 1 0 1>}
  n28 = Subtract(A=@n9)
  n29 = Min(A=Saturate(None=@n3), B=OneMinus(None=Saturate(None=@n28)))
  n30 = OneMinus(None=Abs(None=ConstantBiasScale(None=@n2) {bias=-0.5, scale=2}))
  n31 = DynamicParameter() {parameterIndex=3, defaultValue=<rgba 0 4 1 1>, paramNames [Use Single Color, Color Lerp Contrast, Fresnel Exponent, Opacity Multiplier]}
  n32 = Multiply(A=@n31.Color Lerp Contrast, B=NamedRerouteUsage())
  NamedRerouteDeclaration(None=LinearInterpolate(
    A=MaterialFunctionCall(In=@n29, Contrast=@n6) {materialFunction=&CheapContrast}.Result
    B=MaterialFunctionCall(In=@n30, Contrast=@n32) {materialFunction=&CheapContrast}.Result
    Alpha=Saturate(None=@n31.Use Single Color)
   )) {name=Edges Mask, nodeColor=<rgba 0.5234 1 0 1>}
  NamedRerouteDeclaration(None=@n18) {name=Length Mask, nodeColor=<rgba 0 1 0.6719 1>}
  n33 = DotProduct(A=VertexNormalWS(), B=CameraVectorWS())
  n34 = DynamicParameter() {parameterIndex=3, defaultValue=<rgba 0 4 1 1>, paramNames [Use Single Color, Color Lerp Contrast, Fresnel Exponent, Opacity Multiplier]}
  NamedRerouteDeclaration(None=Power(Base=Abs(None=@n33), Exp=@n34.Fresnel Exponent)) {name=Fresnel Fade, nodeColor=<rgba 0 1 0.9766 1>}
  n35 = DepthFade(FadeDistance=ScalarParameter() {defaultValue=11, parameterName=Depth Fade Distance})
  n36 = Constant() {r=1}
  NamedRerouteDeclaration(None=MaterialFunctionCall("Xb1,PS4"=@n35, "PC-High&&Epic,XSX,PS5"=@n35, "PC-Perf,Android"=@n36, Switch=@n36, "PC-Low(NotPerf)&&Medium"=@n35) {materialFunction=&PlatformSwitch_1D}.Result) {name=Depth Fade, nodeColor=<rgba 0 0.0859 1 1>}
  n37 = ScalarParameter() {defaultValue=33, parameterName=White Blend Offset}
  n38 = Saturate(None=Divide(A=Subtract(A=@n25, B=@n37), B=ScalarParameter() {defaultValue=133, parameterName=White Blend Length}))
  NamedRerouteDeclaration(None=Power(Base=OneMinus(None=@n38), Exp=ScalarParameter() {defaultValue=3, parameterName=White Blend Exponent})) {name=White Blend Alpha, nodeColor=<rgba 1 0 0.8203 1>}
  NamedRerouteDeclaration(None=Saturate(None=Divide(A=@n15, B=@n16))) {name=Base Mask, nodeColor=<rgba 1 0 0.1172 1>}
 ]
}
