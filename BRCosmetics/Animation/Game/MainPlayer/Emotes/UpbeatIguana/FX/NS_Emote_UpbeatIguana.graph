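# Niagara effect for the UpbeatIguana emote, producing a three-color striped ribbon trail, logo glow sprites and a soft light with glow sprite.
# System scripts supply stripe colors, an interpolated trail vector, timeline fade and time-of-day brightness to the emitters.
# Trail slerps four points between target and interpolated vectors, feeding six ribbon renderers with offset stripe positions.
# Glows use fixed per-ID layouts, with separate front-end and gameplay intensities.
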NiagaraSystem /BRCosmetics/Animation/Game/MainPlayer/Emotes/UpbeatIguana/FX/NS_Emote_UpbeatIguana
dependencies {
 usedDataInterfaces [
  &NiagaraDataInterfaceCurve
  &NiagaraDataInterfacePlatformSet
  &NiagaraDataInterfaceArrayFloat3
  &NiagaraDataInterfaceArrayFloat2
 ]
 usedDynamicInputs [
  &Multiply_Int
  &RandomRangeInt
  &SelectLinearColorFromArray
  &SelectIntFromArray
  &Add_Integer
  &Divide_Float
  &Subtract_Float
  &FNTimeOfDayScalar
  &VectorLength
  &SetBoolByFloatComparison
  &Multiply_Vector
 ]
 usedModules [
  &self:SystemSpawnScript.SystemScriptSource.SystemScriptGraph.NiagaraNodeAssignment_4.SetVariables_#1
  &self:SystemSpawnScript.SystemScriptSource.SystemScriptGraph.NiagaraNodeAssignment_3.SetVariables_#2
  &SystemState
  &self:SystemSpawnScript.SystemScriptSource.SystemScriptGraph.NiagaraNodeAssignment_5.SetVariables_#3
  &Timeline
  &/InterpolateTrailVector
  &self:SystemSpawnScript.SystemScriptSource.SystemScriptGraph.NiagaraNodeAssignment_0.SetVariables_#4
 ]
}
emitters {
 LogoGlows {
  emitterSpawnScript [/SetVariables(Emitter.FlatGlowSize:Vector2f=<166 99>) {isSetParameters=true}]
  emitterUpdateScript [
   &EmitterState(
    Inactive Response=::NewEnumerator1
    Life Cycle Mode=::NewEnumerator1
    Loop Behavior=::NewEnumerator1
    Loop Duration Mode=::NewEnumerator1
    UseLoopDelay=false
   )
   &SpawnBurst_Instantaneous(Spawn Count=8, Use Spawn Probability=false, UseLoopCountLimit=false)
  ]
  particleSpawnScript [
   /SetVariables(
    Particles.FlatAlignmentVector:Vector3f=&SelectVectorFromArray(
     Direct Array Index=&ReturnParticleIDAttributes()
     Direct Array Mode=::NewEnumerator1
     Vector Selection Array=DataInterface() {propertyValues {FloatData [<1 -1 0>, <1 0 0>, <1 1 0>, <0 1 0>, <0 0 1>, <0 0 1>, <0 0 1>, <0 0 1>]}}
    )
    Particles.FlatFacingVector:Vector3f=&SelectVectorFromArray(
     Direct Array Index=&ReturnParticleIDAttributes()
     Direct Array Mode=::NewEnumerator1
     Vector Selection Array=DataInterface() {propertyValues {FloatData [<0 0 1>, <0 0 1>, <0 0 1>, <0 0 1>, <-1 1 0>, <1 0 0>, <1 1 0>, <0 1 0>]}}
    )
    Particles.GlowSpriteSize:Vector2f=&SelectVector2DFromArray(
     Direct Array Index=&ReturnParticleIDAttributes()
     Direct Array Mode=::NewEnumerator1
     Vector 2D Selection Array=DataInterface() {propertyValues {FloatData [<133 99>, <133 99>, <133 99>, <133 99>, <166 44>, <166 44>, <166 44>, <166 44>]}}
    )
    Particles.Position=&SelectVectorFromArray(
     Direct Array Index=&ReturnParticleIDAttributes()
     Direct Array Mode=::NewEnumerator1
     Vector Selection Array=DataInterface() {propertyValues {FloatData [<-49 28 -2>, <0 5 -2>, <49 28 -2>, <0 50 -2>]}}
    )
   ) {isSetParameters=true}
   &DynamicMaterialParameters(
    Index 0 Param 1=6
    Index 0 Param 2=0.1
    Index 0 Param 3=2
    Module.Index0Param0Write=true
    Module.Index0Param1Write=true
    Module.Index0Param2Write=true
    Module.Index0Param3Write=true
   )
   &CameraOffset(Camera Offset Amount=5)
  ]
  particleUpdateScript [/SetVariables(Particles.Color=&MakeLinearColorFromVectorAndFloat(
     "Vector (RGB)"=&VectorFromFloat(Value=&FN_FrontEndFloatModifier(Gameplay=$System.TODBrightnessMultiplier, Front End=0.6))
     "Float (Alpha)"=&Multiply_Float(A=$System.FadeInAlpha, B=$System.Timeline.Time)
    )) {isSetParameters=true}]
  rendererClasses [&NiagaraSpriteRendererProperties]
  renderers [{rendererClass=&NiagaraSpriteRendererProperties}]
 }
 SoftLightGlow {
  emitterUpdateScript [
   &EmitterState(
    Inactive Response=::NewEnumerator1
    Life Cycle Mode=::NewEnumerator1
    Loop Behavior=::NewEnumerator1
    Loop Duration Mode=::NewEnumerator1
    UseLoopDelay=false
   )
   &SpawnBurst_Instantaneous(Use Spawn Probability=false, UseLoopCountLimit=false)
   /SetVariables(Emitter.Color:LinearColor=&MakeLinearColorFromVectorAndFloat(
     "Vector (RGB)"=<1 1 1>
     "Float (Alpha)"=&FloatFromCurve(
      FloatCurve=DataInterface() {propertyValues {Curve {keys [{interpMode=RCIM_Cubic, tangentMode=RCTM_Break, time=0, value=0}, <at 0.75 1>]}}}
      CurveIndex=$Engine.System.Age
     )
    )) {isSetParameters=true}
  ]
  particleSpawnScript [
   &InitializeParticle(
    Color Mode=::NewEnumerator0
    Position Offset=<0 33 0>
    Sprite Size Mode=::NewEnumerator3
    Uniform Sprite Size=111
    UsePositionOffset=true
    Write Material Random=true
   )
   &CameraOffset(Camera Offset Amount=66)
   &DynamicMaterialParameters(
    Index 0 Param 1=2
    Index 0 Param 2=1
    Module.Index0Param0Write=true
    Module.Index0Param1Write=true
    Module.Index0Param2Write=true
    Module.Index0Param3Write=true
   )
  ]
  particleUpdateScript [/SetVariables(
    Particles.GlowColor=&MultiplyLinearColor(
     Linear Color=$Emitter.Color
     Scale Factor=&MakeFloat4FromVectorAndFloat(
      "Vector (RGB)"=&VectorFromFloat(Value=$System.TODBrightnessMultiplier)
      "Float (Alpha)"=&FN_FrontEndFloatModifier(Gameplay=0.1, Front End=0.05)
     )
    )
    Particles.LightColor=&MultiplyLinearColor(
     Linear Color=$Emitter.Color
     Scale Factor=&MakeFloat4FromVectorAndFloat("Vector (RGB)"=<5 5 5>, "Float (Alpha)"=&FN_FrontEndFloatModifier(Gameplay=2, Front End=4))
    )
   ) {isSetParameters=true}]
  rendererClasses [&NiagaraLightRendererProperties, &NiagaraSpriteRendererProperties]
  renderers [
   {rendererClass=&NiagaraLightRendererProperties, rendererIndex=0}
   {rendererClass=&NiagaraSpriteRendererProperties, rendererIndex=1}
  ]
 }
 Trail {
  emitterUpdateScript [
   &EmitterState(
    Inactive Response=::NewEnumerator1
    Life Cycle Mode=::NewEnumerator1
    Loop Behavior=::NewEnumerator1
    Loop Duration Mode=::NewEnumerator1
    UseLoopDelay=false
   )
   &SpawnBurst_Instantaneous(Spawn Count=4, Use Spawn Probability=false, UseLoopCountLimit=false)
  ]
  particleSpawnScript [/SetVariables(
    Particles.RibbonWidthLarge=122
    Particles.RibbonWidthSmall:NiagaraFloat=&Lerp_Float(
     A=11
     B=99
     Alpha=&FloatFromCurve(
      FloatCurve=DataInterface() {propertyValues {Curve {keys [{interpMode=RCIM_Cubic, leaveTangent=1.828, tangentMode=RCTM_Break, time=0, value=0}, {arriveTangent=0.0793, time=1, value=1}]}}}
      CurveIndex=&ReturnNormalizedExecIndex()
     )
    )
    Particles.StripeColorA:Vector4f=&MakeVector4_FromLinearColor(LinearColor=$System.StripeColorA)
    Particles.StripeColorB:Vector4f=&MakeVector4_FromLinearColor(LinearColor=$System.StripeColorB)
    Particles.StripeColorC:Vector4f=&MakeVector4_FromLinearColor(LinearColor=$System.StripeColorC)
    Particles.StripeLargeDynamic3:Vector4f=&MakeVector4(
     X=0
     Y=4
     Z=0.5
     W=&FN_PickFloatDependOnPlatform(TargetPlatform=DataInterface() {propertyValues {Platforms {qualityLevelMask=-29}}}, TargetPlatformActive=3, TargetPlatformNoActive=1)
    )
    Particles.StripeSmallDynamic3:Vector4f=&MakeVector4(X=1, Y=0.5, Z=2, W=1)
   ) {isSetParameters=true}]
  particleUpdateScript [
   /SetVariables(Particles.InterpolatedTrailVector:Vector3f=&TransformVector(
     Vector=/Niagara/DynamicInputs/Math/SlerpVector(
      Vector A:Vector3f=$System.TargetVector
      Vector B:Vector3f=$System.InterpolatedVector
      Alpha:NiagaraFloat=&ReturnNormalizedExecIndex()
      Scale:NiagaraFloat=1
     )
     Source Space=::World
     Destination Space=::Simulation
     Apply Scale=true
    )) {isSetParameters=true}
   /SetVariables(
    Particles.Color=&MakeLinearColorFromVectorAndFloat(
     "Vector (RGB)"=&VectorFromFloat(Value=$System.TODBrightnessMultiplier)
     "Float (Alpha)"=&Multiply_Float(A=$System.FadeInAlpha, B=$System.Timeline.Time)
    )
    Particles.InterpolatedPosition:NiagaraPosition=&AddVectorToPosition(
     Position=&AddVectorToPosition(
      Position=/Niagara/DynamicInputs/Helpers/SimulationPosition()
      Vector=&TransformVector(Vector=<0 5 -3>, Destination Space=::Simulation)
     )
     Vector=&Multiply_VectorByFloat(
      Vector=$Particles.InterpolatedTrailVector
      Float=&Multiply_Float(A=&ReturnNormalizedExecIndex(), B=333)
     )
    )
   ) {isSetParameters=true}
   /SetVariables(Particles.StripePositionOffsetVector:Vector3f=&NormalizeVector(
     Vector To Normalize=/Niagara/DynamicInputs/Vectors/CrossProduct(Vector A:Vector3f=$StackContext.InterpolatedTrailVector, Vector B:Vector3f=<0 0 1>, Normalize Incoming Vectors:NiagaraBool=true)
     Safe Normalize=true
     Safe Fallback Threshold=0
     Safe Fallback Vector=<0 0 0.001>
    )) {isSetParameters=true}
   /SetVariables(
    Particles.LargeFlatPositionA:NiagaraPosition=&AddVectorToPosition(Position=$StackContext.InterpolatedPosition, Vector=<0 0 4>)
    Particles.LargeFlatPositionB:NiagaraPosition=&AddVectorToPosition(Position=$StackContext.InterpolatedPosition, Vector=<0 0 2>)
    Particles.LargeFlatPositionC:NiagaraPosition=&AddVectorToPosition(Position=$StackContext.InterpolatedPosition, Vector=<0 0 0>)
    Particles.StripePositionA:NiagaraPosition=&AddVectorToPosition(
     Position=$StackContext.InterpolatedPosition
     Vector=&Multiply_VectorByFloat(Vector=$StackContext.StripePositionOffsetVector, Float=-30)
    )
    Particles.StripePositionC:NiagaraPosition=&AddVectorToPosition(
     Position=$StackContext.InterpolatedPosition
     Vector=&Multiply_VectorByFloat(Vector=$StackContext.StripePositionOffsetVector, Float=30)
    )
   ) {isSetParameters=true}
  ]
  rendererClasses [&NiagaraRibbonRendererProperties]
  renderers [
   {rendererClass=&NiagaraRibbonRendererProperties}
   {rendererClass=&NiagaraRibbonRendererProperties, rendererIndex=1}
   {rendererClass=&NiagaraRibbonRendererProperties, rendererIndex=2}
   {rendererClass=&NiagaraRibbonRendererProperties, rendererIndex=3}
   {rendererClass=&NiagaraRibbonRendererProperties, rendererIndex=4}
   {rendererClass=&NiagaraRibbonRendererProperties, rendererIndex=5}
  ]
 }
}
summary {
 systemName=NS_Emote_UpbeatIguana
 emitters [
  {emitterName=Trail, rendererClasses [&NiagaraRibbonRendererProperties]}
  {emitterName=LogoGlows, rendererClasses [&NiagaraSpriteRendererProperties]}
  {emitterName=SoftLightGlow, rendererClasses [&NiagaraLightRendererProperties, &NiagaraSpriteRendererProperties]}
 ]
 userVariables [
  User.bIsFrontEnd {type=NiagaraBool}
  User.bIsFrontEndPreview {type=NiagaraBool}
  User.FadeInDelay {type=NiagaraFloat, defaultValue=1, description=""}
  User.FadeInDuration {type=NiagaraFloat, defaultValue=1, description=""}
  User.StripeColorArray {
   type=NiagaraDataInterfaceArrayColor
   description=""
   defaultValue {
    struct=NiagaraExt_VariableValue_DataInterface
    value {
     dataInterface=&self:NiagaraDataInterfaceArrayColor_0
     dataInterfaceClass=&NiagaraDataInterfaceArrayColor
    }
   }
  }
  User.StripeColorComboArray {
   type=NiagaraDataInterfaceArrayInt32
   description=""
   defaultValue {
    struct=NiagaraExt_VariableValue_DataInterface
    value {
     dataInterface=&self:NiagaraDataInterfaceArrayInt32_0
     dataInterfaceClass=&NiagaraDataInterfaceArrayInt32
    }
   }
  }
 ]
}
